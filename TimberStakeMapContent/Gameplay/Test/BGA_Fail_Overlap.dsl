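; Test Blueprint actor that plays a fail sound when a locally viewed player dies inside its box volume.
; On begin overlap, non-server instances cache locally viewed Fortnite Athena player pawns in an array.
; End overlap waits one second, then plays the audio component if the cached pawn's health is zero or below, and removes it from the cache.
; Dedicated servers skip both the caching and the sound.
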(event EventBeginPlay)

(event OnComponentBeginOverlap(Box)
  (params (OverlappedComponent :type Primitive%20Component%20Object%20Reference) (OtherActor :type Actor%20Object%20Reference) (OtherComp :type Primitive%20Component%20Object%20Reference) (OtherBodyIndex :type Integer) (bFromSweep :type Boolean) (SweepResult :type Hit%20Result%20Structure%20%28by%20ref%29))
  ; Caches pawns that are locally viewed and inside the collider
  (bind _asfort_player_pawn_athena (Utilities|Casting|CastToFortPlayerPawnAthena OtherActor))
  (if (not (Networking|IsDedicatedServer))
    (Utilities|IsValid _asfort_player_pawn_athena
      (:Is%20Valid
        (if (Pawn|IsLocallyViewed _asfort_player_pawn_athena)
          (Utilities|Array|Add (Variables|Default|GetViewedPawnInsideCollider) _asfort_player_pawn_athena))))))

(event Custom|PlaySoundOnDeath
  ; Play Sound On Death
  (Audio|Components|Audio|Play (Variables|Default|GetAudio) 0.0))

(event OnComponentEndOverlap(Box)
  (params (OverlappedComponent :type Primitive%20Component%20Object%20Reference) (OtherActor :type Actor%20Object%20Reference) (OtherComp :type Primitive%20Component%20Object%20Reference) (OtherBodyIndex :type Integer))
  ; If locally viewed and health == 0, play fail sound. Otherwise clean the cache
  (Utilities|FlowControl|Delay 1.0)
  (bind _asfort_player_pawn_athena (Utilities|Casting|CastToFortPlayerPawnAthena OtherActor))
  (if (Networking|IsDedicatedServer)
    _
    (elif (Utilities|Array|ContainsItem (Variables|Default|GetViewedPawnInsideCollider) _asfort_player_pawn_athena)
      (if (<= (FortPawn|GetHealth _asfort_player_pawn_athena) 0)
        (CallFunction|PlaySoundonDeath)
        (else
          _))
      (Utilities|Array|RemoveItem (Variables|Default|GetViewedPawnInsideCollider) _asfort_player_pawn_athena))))

(fn ConstructionScript)
