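; Timberstake component that plays the battle bus flying sound when the match reaches the BusFlying phase step.
; On BeginPlay it waits for the game state to be ready, then listens on an event router for the stateful game phase step update.
; Both live and saved event states feed GamePhaseStepChanged, so an already-saved step is also handled.
; That function plays BusFlyingSound as 2D audio if valid and ignores other steps.
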(event EventBeginPlay
  (|CallParentFunction)
  (Class|FortAsyncActionGameStateReady|GameStateReadyAsync
    (:OnReady
      (bind _returnvalue (GameplayEventRouter|GetEventRouter "/Script/FortniteGame.FortGameState" (Components|GetOwner)))
      (bind (_asynctaskproxy _payload) (GameplayEventRouter|StartListeningToStatefulEvent _returnvalue "/Script/FortniteGame.FortGamePhaseStepUpdatedEvent")
        (:OnEventReceived
          _)
        (:OnSavedEventStateExists
          _))
      (CallFunction|GamePhaseStepChanged (Utilities|Struct|BreakFortGamePhaseStepUpdatedEvent _payload)))))

(fn GamePhaseStepChanged
  (params (GamePhaseStep :type EAthenaGamePhaseStep%20Enum))
  (bind _busflyingsound (Variables|Default|GetBusFlyingSound))
  (switch Utilities|FlowControl|Switch|SwitchonEAthenaGamePhaseStep GamePhaseStep
    (:BusFlying
      (Utilities|IsValid _busflyingsound
        (:Is%20Valid
          (Audio|PlaySound2D _busflyingsound))))))
