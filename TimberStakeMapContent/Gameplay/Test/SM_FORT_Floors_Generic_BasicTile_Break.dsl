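; Blueprint for a breakable generic floor tile, in a test folder, that damages itself when something overlaps its trigger box.
; On overlap it plays audio, applies initial damage to itself with the overlapping actor as causer, then calls PlayBreakEffects.
; After a variable delay it applies secondary environmental damage to itself, giving a two-stage break.
; The overlap has no actor filter, and the construction script only calls its parent.
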(event OnComponentBeginOverlap(Box)
  (params (OverlappedComponent :type Primitive%20Component%20Object%20Reference) (OtherActor :type Actor%20Object%20Reference) (OtherComp :type Primitive%20Component%20Object%20Reference) (OtherBodyIndex :type Integer) (bFromSweep :type Boolean) (SweepResult :type Hit%20Result%20Structure%20%28by%20ref%29))
  (Audio|Components|Audio|Play (Variables|Default|GetAudio) 0.0)
  (Game|Damage|ApplyDamage self (Variables|Default|GetInitial%20Damage) 0 OtherActor)
  (CallFunction|PlayBreakEffects)
  (Utilities|FlowControl|Delay (Variables|Default|GetDelay))
  (Game|Damage|ApplyDamage self (Variables|Default|GetSecondary%20Damage) 0 0 "/Engine/EngineDamageTypes/DmgTypeBP_Environmental.DmgTypeBP_Environmental_C"))

(fn ConstructionScript
  (|CallParentFunction))
